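# Niagara effect for a FinFlake cosmetic companion antic: slow mesh bubbles that drift and then pop.
# Two bubble emitters spawn at a rate tapering off by mid-cycle, moved by curve-driven gravity and drag, with a vortex on the first.
# Bubbles write death positions to a system array through scratch modules, a CPU-style handoff alongside a system-level ReadDeath module.
# NE_Pop uses the resulting spawn flag and position to burst short-lived, fading mesh bits.
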NiagaraSystem /CosmeticCompanions/FX/Builtins/FinFlake/Antic/NS_Antic_FinFlake_SlowBubbles
dependencies {
 usedDataInterfaces [&NiagaraDataInterfaceCurve, &NiagaraDataInterfaceArrayPosition]
 usedDynamicInputs []
 usedModules [
  &self:SystemSpawnScript.SystemScriptSource.SystemScriptGraph.NiagaraNodeAssignment_4.SetVariables_#1
  &SystemState
  &self:ReadDeath_01
  &self:SystemSpawnScript.SystemScriptSource.SystemScriptGraph.NiagaraNodeAssignment_5.SetVariables_#2
 ]
}
emitters {
 NE_Bubbles_01 {
  emitterUpdateScript [
   &EmitterState(Inactive Response=::NewEnumerator1, Life Cycle Mode=::NewEnumerator1, Loop Behavior=::NewEnumerator1, Loop Duration=4, UseLoopDelay=false)
   &SpawnRate(
    SpawnRate=&FloatFromCurve(
     FloatCurve=DataInterface() {propertyValues {Curve {keys [
         {interpMode=RCIM_Cubic, tangentMode=RCTM_User, time=0, value=1}
         {arriveTangent=-4.5539, interpMode=RCIM_Cubic, leaveTangent=-4.5539, tangentMode=RCTM_User, time=0.5, value=0}
        ]}}}
     CurveIndex=$Emitter.NormalizedLoopAge
     Scale Curve=7.5
    )
    Use Spawn Probability=false
   )
  ]
  particleSpawnScript [
   &InitializeParticle(
    Color=<rgba 1 1 1 0.1>
    Lifetime Max=1.5
    Lifetime Min=0.75
    Lifetime Mode=::NewEnumerator1
    Mesh Scale Mode=::NewEnumerator1
    Mesh Uniform Scale Max=0.04
    Mesh Uniform Scale Min=0.02
    Position Offset=<-4 0 3>
    UsePositionOffset=true
    Write Material Random=true
   )
   &ShapeLocation(
    Handle Radius=4.5
    Large Radius=4.5
    MaskSpawnGroup=false
    Override Seed=false
    Rotation Angle=0.25
    Rotation Axis=<0 0 0>
    Rotation Mode=::NewEnumerator1
    Shape Primitive=::NewEnumerator3
    U Distribution=0.5
    UseOwnerScale=true
    V Distribution=1
   ) {module=ShapeLocation001}
   &AddVelocity(Velocity=&RandomRangeVector(Minimum=<1 0 0>, Maximum=<2 0 0>), Velocity Speed Scale=30)
   &CameraOffset(Camera Offset Amount=5)
  ]
  particleUpdateScript [
   &ParticleState()
   &GravityForce(
    Coordinate Space=::Local
    Gravity=&MakeVector(
     X=0
     Y=&FloatFromCurve(
      FloatCurve=DataInterface() {propertyValues {Curve {
         defaultValue=1
         keys [
          {interpMode=RCIM_Cubic, tangentMode=RCTM_User, time=0, value=0}
          {arriveTangent=0.0593, interpMode=RCIM_Cubic, leaveTangent=0.0593, tangentMode=RCTM_User, time=0.5, value=0}
          {interpMode=RCIM_Cubic, tangentMode=RCTM_User, time=1, value=1}
         ]
        }}}
      Scale Curve=-250
     )
     Z=&RandomRangeFloat(
      Minimum=&FloatFromCurve(
       FloatCurve=DataInterface() {propertyValues {Curve {keys [
           {interpMode=RCIM_Cubic, tangentMode=RCTM_User, time=0, value=0}
           {interpMode=RCIM_Cubic, tangentMode=RCTM_User, time=0.25, value=-2}
           {interpMode=RCIM_Cubic, tangentMode=RCTM_User, time=0.5, value=2}
           {interpMode=RCIM_Cubic, tangentMode=RCTM_User, time=1, value=2}
          ]}}}
       Scale Curve=200
      )
      Maximum=&FloatFromCurve(
       FloatCurve=DataInterface() {propertyValues {Curve {keys [
           {interpMode=RCIM_Cubic, tangentMode=RCTM_User, time=0, value=0}
           {interpMode=RCIM_Cubic, tangentMode=RCTM_User, time=0.25, value=-1.25}
           {interpMode=RCIM_Cubic, tangentMode=RCTM_User, time=0.5, value=3}
           {interpMode=RCIM_Cubic, tangentMode=RCTM_User, time=1, value=3}
          ]}}}
       Scale Curve=200
      )
     )
    )
   )
   &VortexForce(Origin Pull Amount=250, UseOriginOffset=false, Vortex Force Amount=500)
   &Drag(Drag=15, Use Linear Drag=true, Use Rotational Drag=true)
   &SolveForcesAndVelocity(Clamp Velocity=false, Limit Acceleration=false)
   &DynamicMaterialParameters(
    Index 3 Param 2=&MakeFloatFromVector(Vector=$Particles.Scale, Channel=::NewEnumerator0)
    Index 3 Param 4=&RandomRangeFloat(Minimum=0, Maximum=100)
    Module.Param0WriteEnabled3=false
    Module.Param1WriteEnabled3=true
    Module.Param2WriteEnabled3=false
    Module.Param3WriteEnabled3=true
    Write Parameter Index 0=false
    Write Parameter Index 3=true
   )
   &ScaleMeshSize(Scale Factor=&Multiply_VectorByFloat(
     Vector=<1 1 1>
     Float=&FloatFromCurve(FloatCurve=DataInterface() {propertyValues {Curve {keys [<at 0 0>,<at 0.2 1>,<at 0.6969 1>,<at 0.7969 1.5>,<at 0.9994 0>]}}})
    ))
   &self:WriteDeath_02(Position Array:NiagaraDataInterfaceArrayPosition=$System.DeathArray) {module=WriteDeath_02}
  ]
  rendererClasses [&NiagaraMeshRendererProperties]
  renderers [{rendererClass=&NiagaraMeshRendererProperties, rendererIndex=0}]
 }
 NE_Bubbles_02 {
  emitterUpdateScript [
   &EmitterState(Inactive Response=::NewEnumerator1, Life Cycle Mode=::NewEnumerator1, Loop Behavior=::NewEnumerator1, Loop Duration=4, UseLoopDelay=false)
   &SpawnRate(
    SpawnRate=&FloatFromCurve(
     FloatCurve=DataInterface() {propertyValues {Curve {keys [
         {interpMode=RCIM_Cubic, tangentMode=RCTM_User, time=0, value=1}
         {arriveTangent=-4.5539, interpMode=RCIM_Cubic, leaveTangent=-4.5539, tangentMode=RCTM_User, time=0.5, value=0}
        ]}}}
     CurveIndex=$Emitter.NormalizedLoopAge
     Scale Curve=7.5
    )
    Use Spawn Probability=false
   )
  ]
  particleSpawnScript [
   &InitializeParticle(
    Color=<rgba 1 1 1 0.1>
    Lifetime Max=1.5
    Lifetime Min=0.75
    Lifetime Mode=::NewEnumerator1
    Mesh Scale Mode=::NewEnumerator1
    Mesh Uniform Scale Max=0.03
    Mesh Uniform Scale Min=0.015
    Position Offset=<-4 0 1>
    UsePositionOffset=true
    Write Material Random=true
   )
   &ShapeLocation(
    MaskSpawnGroup=false
    Rotation Angle=0.25
    Rotation Axis=<0 0 0>
    Rotation Mode=::NewEnumerator1
    Sphere Distribution=::NewEnumerator1
    Sphere Radius=1
    UseOwnerScale=true
    V Position=3
   ) {module=ShapeLocation001}
   &AddVelocity(Velocity=&RandomRangeVector(Minimum=<2.5 0 0>, Maximum=<5 0 0>), Velocity Speed Scale=30)
   &CameraOffset(Camera Offset Amount=5)
  ]
  particleUpdateScript [
   &ParticleState()
   &GravityForce(
    Coordinate Space=::Local
    Gravity=&MakeVector(
     X=0
     Y=&FloatFromCurve(
      FloatCurve=DataInterface() {propertyValues {Curve {
         defaultValue=1
         keys [
          {interpMode=RCIM_Cubic, tangentMode=RCTM_User, time=0, value=0}
          {arriveTangent=0.0593, interpMode=RCIM_Cubic, leaveTangent=0.0593, tangentMode=RCTM_User, time=0.5, value=0}
          {interpMode=RCIM_Cubic, tangentMode=RCTM_User, time=1, value=1}
         ]
        }}}
      Scale Curve=-250
     )
     Z=&RandomRangeFloat(
      Minimum=&FloatFromCurve(
       FloatCurve=DataInterface() {propertyValues {Curve {keys [
           {interpMode=RCIM_Cubic, tangentMode=RCTM_User, time=0, value=0}
           {interpMode=RCIM_Cubic, tangentMode=RCTM_User, time=0.25, value=-2}
           {interpMode=RCIM_Cubic, tangentMode=RCTM_User, time=0.5, value=2}
           {interpMode=RCIM_Cubic, tangentMode=RCTM_User, time=1, value=2}
          ]}}}
       Scale Curve=200
      )
      Maximum=&FloatFromCurve(
       FloatCurve=DataInterface() {propertyValues {Curve {keys [
           {interpMode=RCIM_Cubic, tangentMode=RCTM_User, time=0, value=0}
           {interpMode=RCIM_Cubic, tangentMode=RCTM_User, time=0.25, value=-1.25}
           {interpMode=RCIM_Cubic, tangentMode=RCTM_User, time=0.5, value=3}
           {interpMode=RCIM_Cubic, tangentMode=RCTM_User, time=1, value=3}
          ]}}}
       Scale Curve=200
      )
     )
    )
   )
   &Drag(Drag=15, Use Linear Drag=true, Use Rotational Drag=true)
   &SolveForcesAndVelocity(Clamp Velocity=false, Limit Acceleration=false)
   &DynamicMaterialParameters(
    Index 3 Param 2=&MakeFloatFromVector(Vector=$Particles.Scale, Channel=::NewEnumerator0)
    Index 3 Param 4=&RandomRangeFloat(Minimum=0, Maximum=100)
    Module.Param0WriteEnabled3=false
    Module.Param1WriteEnabled3=true
    Module.Param2WriteEnabled3=false
    Module.Param3WriteEnabled3=true
    Write Parameter Index 0=false
    Write Parameter Index 3=true
   )
   &ScaleMeshSize(Scale Factor=&Multiply_VectorByFloat(
     Vector=<1 1 1>
     Float=&FloatFromCurve(FloatCurve=DataInterface() {propertyValues {Curve {keys [<at 0 0>,<at 0.2 1>,<at 0.6969 1>,<at 0.7969 1.5>,<at 0.9994 0>]}}})
    ))
   &self:WriteDeath_02(Position Array:NiagaraDataInterfaceArrayPosition=$System.DeathArray) {module=WriteDeath_02}
  ]
  rendererClasses [&NiagaraMeshRendererProperties]
  renderers [{rendererClass=&NiagaraMeshRendererProperties, rendererIndex=0}]
 }
 NE_Pop {
  emitterUpdateScript [
   &EmitterState(Life Cycle Mode=::NewEnumerator1, Loop Behavior=::NewEnumerator1, UseLoopDelay=false)
   &SpawnPerFrame(Spawn:NiagaraBool=$System.DeathSpawn, Spawn Count=3, Use Spawn Probability=false)
  ]
  particleSpawnScript [
   &InitializeParticle(
    Color=<rgba 1 1 1 0.5>
    Lifetime Max=0.3
    Lifetime Min=0.1
    Lifetime Mode=::NewEnumerator1
    Mesh Scale Mode=::NewEnumerator3
    Mesh Uniform Scale=0.01
    Position=$System.DeathPosition
    Position Mode=::NewEnumerator1
    UsePositionOffset=false
    Write Material Random=true
   )
   &SphereLocation(
    MaskSpawnGroup=false
    Override Seed=false
    Sphere Radius=1
    Surface Only=false
    UseNonUniformScale=false
    UseOffset=false
    UseSphereOrientation=false
   )
   &AddVelocity(Velocity=&RandomRangeVector(Minimum=<-90 -90 -90>, Maximum=<90 90 90>))
  ]
  particleUpdateScript [
   &ParticleState()
   &Drag(Drag=30, Use Linear Drag=true, Use Rotational Drag=true)
   &SolveForcesAndVelocity(Clamp Velocity=false, Limit Acceleration=false)
   &DynamicMaterialParameters(
    Index 3 Param 2=&MakeFloatFromVector(Vector=$Particles.Scale, Channel=::NewEnumerator0)
    Index 3 Param 4=&RandomRangeFloat(Minimum=0, Maximum=100)
    Module.Param0WriteEnabled3=false
    Module.Param1WriteEnabled3=true
    Module.Param2WriteEnabled3=false
    Module.Param3WriteEnabled3=true
    Write Parameter Index 0=false
    Write Parameter Index 3=true
   )
   &ScaleColor(
    Scale Alpha=&Multiply_Float(
     A=&FloatFromCurve(FloatCurve=DataInterface() {propertyValues {Curve {keys [
          {interpMode=RCIM_Cubic, tangentMode=RCTM_User, time=0, value=0}
          {interpMode=RCIM_Cubic, tangentMode=RCTM_User, time=0.3, value=1}
          {tangentMode=RCTM_User, time=1, value=0}
         ]}}})
     B=1
    )
    Scale RGB=&Multiply_VectorByFloat(
     Vector=<1 1 1>
     Float=&FloatFromCurve(FloatCurve=DataInterface() {propertyValues {Curve {keys [
          {interpMode=RCIM_Cubic, tangentMode=RCTM_User, time=0, value=0}
          {interpMode=RCIM_Cubic, tangentMode=RCTM_User, time=0.3, value=1}
          {tangentMode=RCTM_User, time=1, value=0}
         ]}}})
    )
    ScaleA=true
    ScaleRGB=true
   )
   &ScaleMeshSize(Scale Factor=&Multiply_VectorByFloat(
     Vector=<1 1 1>
     Float=&FloatFromCurve(FloatCurve=DataInterface() {propertyValues {Curve {defaultValue=0.75, keys [{tangentMode=RCTM_User, time=0, value=1}, {tangentMode=RCTM_User, time=1, value=0}]}}})
    ))
  ]
  rendererClasses [&NiagaraMeshRendererProperties]
  renderers [{rendererClass=&NiagaraMeshRendererProperties, rendererIndex=0}]
 }
}
summary {
 systemName=NS_Antic_FinFlake_SlowBubbles
 emitters [
  {emitterName=NE_Bubbles_01, rendererClasses [&NiagaraMeshRendererProperties]}
  {emitterName=NE_Pop, rendererClasses [&NiagaraMeshRendererProperties]}
  {emitterName=NE_Bubbles_02, rendererClasses [&NiagaraMeshRendererProperties]}
 ]
 userVariables [
  User.bIsFrontEnd {type=NiagaraBool}
  User.bIsFrontEndPreview {type=NiagaraBool}
 ]
}
